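# Material function for the Season 40 Reload pickaxe that outputs hard-switched triplanar UVs from Position and Normal inputs, defaulting to world position and vertex normal.
# Each absolute normal component is compared with the larger of the other two, using Floor and Saturate for a 0/1 dominant-axis mask.
# The masks select the YZ, XZ or XY position plane and sum into one UV set, with no blending between projections.
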MaterialFunction /BRCosmetics/Weapons/FORT_Melee/Pickaxe_Season_40Reload/Materials/MF_Season40PickaxeTriplanarUVs
graph {nodes [
  n1 = FunctionInput(Preview=WorldPosition().XYZ) {bUsePreviewValueAsDefault=true, inputName=Position}
  n2 = Abs(None=FunctionInput(Preview=VertexNormalWS()) {bUsePreviewValueAsDefault=true, inputName=Normal})
  n3 = ComponentMask(None=@n2) {g=true}
  n4 = ComponentMask(None=@n2) {b=true}
  n5 = ComponentMask(None=@n2) {r=true}
  n6 = Add(A=OneMinus(None=Max(A=@n3, B=@n4)), B=@n5)
  n7 = Multiply(A=ComponentMask(None=@n1) {b=true, g=true}, B=Saturate(None=Floor(None=@n6)))
  n8 = Add(A=OneMinus(None=Max(A=@n5, B=@n4)), B=@n3)
  n9 = Multiply(A=ComponentMask(None=@n1) {b=true, r=true}, B=Saturate(None=Floor(None=@n8)))
  n10 = ComponentMask(None=@n1) {g=true, r=true}
  n11 = OneMinus(None=Max(A=@n5, B=@n3))
  n12 = Saturate(None=Floor(None=Add(A=@n11, B=@n4)))
  FunctionOutput(None=Add(A=Add(A=@n7, B=@n9), B=Multiply(A=@n10, B=@n12))) {outputName=Triplanar UVs}
 ]}
