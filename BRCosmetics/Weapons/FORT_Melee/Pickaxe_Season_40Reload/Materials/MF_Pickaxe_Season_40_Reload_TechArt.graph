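# Material function that adds animated tech-art effects to the Season 40 Reload pickaxe, overriding incoming Emissive, BaseColor and Normal through mask-driven blends.
# It pans packed-texture channels over triplanar pre-skinned local-position UVs, tinting each with three-color lerps for detail and highlights.
# An animated grid supplies colors, lines and random cells, alongside UV distortion and a looping expanding band.
# Scaled local-space sphere masks provide gradients, one pulsing with a time-driven sine.
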MaterialFunction /BRCosmetics/Weapons/FORT_Melee/Pickaxe_Season_40Reload/Materials/MF_Pickaxe_Season_40_Reload_TechArt
graph {nodes [
  n1 = GetMaterialAttributes(None=NamedRerouteUsage()) {attributeGetTypes [#1, #2, #3]}
  FunctionOutput(None=SetMaterialAttributes!(MaterialAttributes=@n1.MaterialAttributes, EmissiveColor=NamedRerouteUsage(), BaseColor=NamedRerouteUsage(), Normal=NamedRerouteUsage()) {attributeSetTypes [#1, #2, #3]})
  n2 = ComponentMask(None=NamedRerouteUsage()) {g=true}
  Multiply(A=@n2, B=Saturate(None=NamedRerouteUsage()))
  NamedRerouteDeclaration(None=@n1.EmissiveColor) {name=In_Emissive, nodeColor=<rgba 0 1 0.7188 1>}
  n3 = NamedRerouteUsage()
  n4 = NamedRerouteUsage()
  n5 = NamedRerouteUsage()
  n6 = Multiply(A=@n4, B=Add(B=Multiply(A=@n5) {constB=2}) {constA=0.5})
  n7 = NamedRerouteUsage()
  n8 = Multiply(A=NamedRerouteUsage(), B=Multiply(A=NamedRerouteUsage(), B=OneMinus(None=@n7)))
  n9 = NamedRerouteUsage()
  n10 = Add(
   A=LinearInterpolate(A=@n4, B=Add(A=@n6, B=@n8), Alpha=Multiply(A=@n9) {constB=0.95})
   B=Multiply(A=NamedRerouteUsage(), B=NamedRerouteUsage())
  )
  n11 = NamedRerouteUsage()
  NamedRerouteDeclaration(None=Add(A=LinearInterpolate(A=@n3, B=@n10, Alpha=@n2), B=Multiply(A=@n2, B=@n11))) {name=Out_Emissive, nodeColor=<rgba 1 0.3984 0 1>}
  NamedRerouteDeclaration(None=MaterialFunctionCall(
    A=VectorParameter() {parameterName=GradientLowColor, defaultValue=<rgba 0.1302 0.1302 0.1302 1>}.RGB
    B=VectorParameter() {parameterName=GradientMedColor, defaultValue=<rgba 0.0011 0.0822 0.2083 1>}.RGB
    C=VectorParameter() {parameterName=GradientHighColor, defaultValue=<rgba 0.0373 0.4118 0.6458 1>}.RGB
    Alpha=NamedRerouteUsage()
   ) {materialFunction=&Lerp_3Color}.Result) {name=GradientColors, nodeColor=<rgba 0.3594 1 0 1>}
  n12 = MaterialFunctionCall!(
   UV=NamedRerouteUsage()
   Tiling
   Tiling X=ScalarParameter() {defaultValue=1, parameterName=AnimatedGridXTiling}
   Tiling Y=ScalarParameter() {defaultValue=1, parameterName=AnimatedGridYTiling}
   Color 01=VectorParameter() {parameterName="Color 01 (V3)", defaultValue=<rgba 0.0721 0.3529 0.7083 1>}.RGB
   Color 02=VectorParameter() {parameterName="Color 02 (V3)", defaultValue=<rgba 0.0172 0.0716 0.1302 1>}.RGB
   Speed
   Speed Random Amount
   Line Thickness=Constant() {r=0.05}
   Line Hardness
   Line Color=VectorParameter() {parameterName="Line Color (V3)", defaultValue=<rgba 0.0277 0.6771 0.6771 1>}.RGB
  ) {materialFunction=/Game/Global/MaterialFunctions/Cosmetics/AnimatedGrid/MF_AnimatedGrid}
  NamedRerouteDeclaration(None=@n12."Result (V3)") {name=AnimatedGridColors, nodeColor=<rgba 1 0 0.6797 1>}
  n13 = ScalarParameter() {defaultValue=0.01, parameterName=XSpeedRedChannel}
  n14 = ScalarParameter() {defaultValue=0.01, parameterName=YSpeedRedChannel}
  n15 = TextureSample(UVs=Panner(Coordinate=NamedRerouteUsage(), Time=Time(), Speed=AppendVector(A=@n13, B=@n14)), Tex=NamedRerouteUsage()) {
   samplerSource=SSM_Wrap_WorldGroupSettings
   samplerType=SAMPLERTYPE_LinearColor
   texture=/BRCosmetics/Weapons/FORT_Melee/Pickaxe_Season_40Reload/Textures/T_Pickaxe_Season40_Reload_TA
  }
  NamedRerouteDeclaration(None=@n15.R) {name=RedChannelAnimated, nodeColor=<rgba 1 0 0 1>}
  n16 = VertexInterpolator(VS=LocalPosition() {localOrigin=InstancePreSkinning}.XYZ)
  n17 = Sine(None=Multiply(A=Time()) {constB=0.25})
  n18 = Multiply(
   A=MaterialFunctionCall(
    R=ScalarParameter() {defaultValue=1, group=Season40PickaxeReloaded, parameterName=GradientXScale}
    G=ScalarParameter() {defaultValue=4, group=Season40PickaxeReloaded, parameterName=GradientYScale}
    B=ScalarParameter() {defaultValue=0.616, group=Season40PickaxeReloaded, parameterName=GradientZScale}
   ) {materialFunction=&AppendMany}.RGB
   B=Add(B=Multiply(A=@n17) {constB=0.15}) {constA=1}
  )
  n19 = TransformPosition(None=ObjectPositionWS()) {transformSourceType=TRANSFORMPOSSOURCE_World}
  n20 = VectorParameter() {parameterName=GradientPositionOffset, defaultValue=<rgba 0 41.7728 14.1811 0>}
  NamedRerouteDeclaration(None=SphereMask(
    A=Multiply(A=@n16.PS, B=@n18)
    B=Add(A=@n19, B=@n20.RGB)
    Radius=ScalarParameter() {defaultValue=39.3387, parameterName=GradientMaskRadius}
    Hardness=ScalarParameter() {defaultValue=0.5476, parameterName=GradientMaskHardness}
   )) {name=GreyscaleGradientMask, nodeColor=<rgba 0 0.2311 1 1>}
  n21 = ScalarParameter() {defaultValue=0.02, parameterName=XSpeedGreenChannel}
  n22 = ScalarParameter() {defaultValue=-0.01, parameterName=YSpeedGreenChannel}
  n23 = TextureSample(UVs=Panner(Coordinate=NamedRerouteUsage(), Time=Time(), Speed=AppendVector(A=@n21, B=@n22)), Tex=NamedRerouteUsage()) {
   samplerSource=SSM_Wrap_WorldGroupSettings
   samplerType=SAMPLERTYPE_LinearColor
   texture=&T_Pickaxe_Season40_Reload_TA
  }
  NamedRerouteDeclaration(None=@n23.G) {name=GreenChannelAnimated, nodeColor=<rgba 0 1 0 1>}
  n24 = ScalarParameter() {defaultValue=-0.04, parameterName=XSpeedBlueChannel}
  n25 = ScalarParameter() {parameterName=YSpeedBlueChannel}
  n26 = TextureSample(UVs=Panner(Coordinate=NamedRerouteUsage(), Time=Time(), Speed=AppendVector(A=@n24, B=@n25)), Tex=NamedRerouteUsage()) {
   samplerSource=SSM_Wrap_WorldGroupSettings
   samplerType=SAMPLERTYPE_LinearColor
   texture=&T_Pickaxe_Season40_Reload_TA
  }
  NamedRerouteDeclaration(None=@n26.B) {name=BlueChannelAnimated, nodeColor=<rgba 0 0 1 1>}
  n27 = TextureObjectParameter() {
   parameterName=PackedAnimatedDetails
   samplerType=SAMPLERTYPE_LinearColor
   texture=&T_Pickaxe_Season40_Reload_TA
  }
  NamedRerouteDeclaration(None=@n27) {name=PackedTexture, nodeColor=<rgba 0.2422 0 1 1>}
  NamedRerouteDeclaration(None=VectorParameter() {parameterName=RedChannelLowColor}.RGB) {name=RedChannelLowColor, nodeColor=<rgba 0.2396 0.0114 0 1>}
  NamedRerouteDeclaration(None=VectorParameter() {parameterName=RedChannelMidColor, defaultValue=<rgba 0.0081 0.0406 0.1042 1>}.RGB) {name=RedChannelMidColor, nodeColor=<rgba 0.5104 0.0148 0.005 1>}
  NamedRerouteDeclaration(None=VectorParameter() {parameterName=RedChannelHighColor, defaultValue=<rgba 0.0015 0.028 0.0417 1>}.RGB) {name=RedChannelHighColor, nodeColor=<rgba 1 0.0133 0 1>}
  NamedRerouteDeclaration(None=FunctionInput() {inputName=In_MaterialAttributes, inputType=FunctionInput_MaterialAttributes}) {name=In_MaterialAttributes, nodeColor=<rgba 0.0781 0 1 1>}
  NamedRerouteDeclaration(None=VectorParameter() {parameterName=GreenChannelLowColor}.RGB) {name=GreenChannelLowColor, nodeColor=<rgba 0.0035 0.2396 0.015 1>}
  NamedRerouteDeclaration(None=VectorParameter() {parameterName=GreenChannelMidColor, defaultValue=<rgba 0.004 0.0089 0.0609 1>}.RGB) {name=GreenChannelMidColor, nodeColor=<rgba 0.0328 0.5104 0.0194 1>}
  NamedRerouteDeclaration(None=VectorParameter() {parameterName=GreenChannelHighColor, defaultValue=<rgba 0.0019 0.0211 0.04 1>}.RGB) {name=GreenChannelHighColor, nodeColor=<rgba 0.035 1 0.0505 1>}
  NamedRerouteDeclaration(None=VectorParameter() {parameterName=BlueChannelLowColor}.RGB) {name=BlueChannelLowColor, nodeColor=<rgba 0.0042 0 0.2396 1>}
  NamedRerouteDeclaration(None=VectorParameter() {parameterName=BlueChannelMidColor, defaultValue=<rgba 0.0012 0.011 0.0312 1>}.RGB) {name=BlueChannelMidColor, nodeColor=<rgba 0.0402 0.005 0.5104 1>}
  NamedRerouteDeclaration(None=VectorParameter() {parameterName=BlueChannelHighColor, defaultValue=<rgba 0.0044 0.031 0.0522 1>}.RGB) {name=BlueChannelHighColor, nodeColor=<rgba 0.0696 0 1 1>}
  NamedRerouteDeclaration(None=MaterialFunctionCall(A=NamedRerouteUsage(), B=NamedRerouteUsage(), C=NamedRerouteUsage(), Alpha=NamedRerouteUsage()) {materialFunction=&Lerp_3Color}.Result) {name=AnimatedColoredRedChannel, nodeColor=<rgba 1 0.018 0 1>}
  NamedRerouteDeclaration(None=MaterialFunctionCall(A=NamedRerouteUsage(), B=NamedRerouteUsage(), C=NamedRerouteUsage(), Alpha=NamedRerouteUsage()) {materialFunction=&Lerp_3Color}.Result) {name=AnimatedColoredGreenChannel, nodeColor=<rgba 0.1242 1 0.0027 1>}
  NamedRerouteDeclaration(None=MaterialFunctionCall(A=NamedRerouteUsage(), B=NamedRerouteUsage(), C=NamedRerouteUsage(), Alpha=NamedRerouteUsage()) {materialFunction=&Lerp_3Color}.Result) {name=AnimatedColoredBlueChannel, nodeColor=<rgba 0 0.42 1 1>}
  NamedRerouteDeclaration(None=@n1.Normal) {name=In_Normal, nodeColor=<rgba 0 1 0.0625 1>}
  n28 = TextureSample() {samplerType=SAMPLERTYPE_LinearColor, texture=/BRCosmetics/Weapons/FORT_Melee/Pickaxe_Season_40Reload/Textures/T_Season_40Reload_Pickaxe_FX}
  NamedRerouteDeclaration(None=FunctionInput(Preview=@n28.RGB) {inputName=In_TechArtMask, sortPriority=1}) {name=In_TechArtMask, nodeColor=<rgba 0.2422 1 0 1>}
  n29 = VertexInterpolator(VS=LocalPosition() {localOrigin=InstancePreSkinning}.XYZ)
  n30 = MaterialFunctionCall(
   R=ScalarParameter() {defaultValue=1, group=Season40PickaxeReloaded, parameterName=SphereMaskXScale}
   G=ScalarParameter() {defaultValue=1.5, group=Season40PickaxeReloaded, parameterName=SphereMaskYScale}
   B=ScalarParameter() {defaultValue=0.1887, group=Season40PickaxeReloaded, parameterName=SphereMaskZScale}
  ) {materialFunction=&AppendMany}
  n31 = TransformPosition(None=ObjectPositionWS()) {transformSourceType=TRANSFORMPOSSOURCE_World}
  n32 = VectorParameter() {parameterName=SphereMaskPositionOffset, defaultValue=<rgba 0 7.2966 2.5311 0>}
  NamedRerouteDeclaration(None=SphereMask(
    A=Multiply(A=@n29.PS, B=@n30.RGB)
    B=Add(A=@n31, B=@n32.RGB)
    Radius=ScalarParameter() {defaultValue=12, parameterName=SphereMaskRadius}
    Hardness=ScalarParameter() {defaultValue=0.5, parameterName=SphereMaskHardness}
   )) {name=SphereMask, nodeColor=<rgba 0 0.2311 1 1>}
  n33 = VertexNormalWS()
  n34 = VertexInterpolator(VS=LocalPosition() {localOrigin=InstancePreSkinning}.XYZ)
  NamedRerouteDeclaration(None=MaterialFunctionCall!(
    Normal=Transform(None=@n33) {transformSourceType=TRANSFORMSOURCE_World, transformType=TRANSFORM_Local}
    Position=Multiply(A=@n34.PS) {constB=0.15}
   ) {materialFunction=&/MF_Season40PickaxeTriplanarUVs}.Triplanar UVs) {name=TriPlanarUVForPackedTextures, nodeColor=<rgba 0.5469 0 1 1>}
  n35 = NamedRerouteUsage()
  n36 = NamedRerouteUsage()
  NamedRerouteDeclaration(None=Add(A=Add(A=@n35, B=@n36), B=NamedRerouteUsage())) {name=AnimatedColoredChannels, nodeColor=<rgba 1 0 0.4922 1>}
  NamedRerouteDeclaration(None=@n12.Gird Lines) {name=GridLines, nodeColor=<rgba 1 0.9375 0 1>}
  n37 = Multiply(A=NamedRerouteUsage()) {constB=0.001}
  n38 = Multiply(A=NamedRerouteUsage()) {constB=-0.001}
  n39 = NamedRerouteUsage()
  NamedRerouteDeclaration(None=AppendVector(A=Add(A=@n37, B=@n38), B=Multiply(A=@n39) {constB=0.002})) {name=UVDistortionAmount, nodeColor=<rgba 0.6172 0 1 1>}
  n40 = ScalarParameter() {defaultValue=0.06, parameterName=XSpeedHighlightChannel}
  n41 = ScalarParameter() {defaultValue=-0.04, parameterName=YSpeedHighlightChannel}
  n42 = TextureSample(UVs=Panner(Coordinate=NamedRerouteUsage(), Time=Time(), Speed=AppendVector(A=@n40, B=@n41)), Tex=NamedRerouteUsage()) {
   samplerSource=SSM_Wrap_WorldGroupSettings
   samplerType=SAMPLERTYPE_LinearColor
   texture=&T_Pickaxe_Season40_Reload_TA
  }
  NamedRerouteDeclaration(None=@n42.B) {name=AnimatedHighlights, nodeColor=<rgba 1 0 0 1>}
  NamedRerouteDeclaration(None=VectorParameter() {parameterName=HighlightLowColor}.RGB) {name=HighlightColorLow, nodeColor=<rgba 0.0042 0 0.2396 1>}
  NamedRerouteDeclaration(None=VectorParameter() {parameterName=HighlightMidColor, defaultValue=<rgba 0.0096 0.0845 0.2396 1>}.RGB) {name=HighlightColorMid, nodeColor=<rgba 0.0402 0.005 0.5104 1>}
  NamedRerouteDeclaration(None=VectorParameter() {parameterName=HighlightHighColor, defaultValue=<rgba 0.0673 0.6662 0.8 1>}.RGB) {name=HighlightColorHigh, nodeColor=<rgba 0.0696 0 1 1>}
  NamedRerouteDeclaration(None=MaterialFunctionCall(A=NamedRerouteUsage(), B=NamedRerouteUsage(), C=NamedRerouteUsage(), Alpha=NamedRerouteUsage()) {materialFunction=&Lerp_3Color}.Result) {name=HighlightColor, nodeColor=<rgba 0 0.42 1 1>}
  n43 = VertexNormalWS()
  n44 = VertexInterpolator(VS=LocalPosition() {localOrigin=InstancePreSkinning}.XYZ)
  n45 = Constant3Vector() {constant=<rgba 0.5 0.01 0.1 0>}
  NamedRerouteDeclaration(None=MaterialFunctionCall(
    Normal=Transform(None=@n43) {transformSourceType=TRANSFORMSOURCE_World, transformType=TRANSFORM_Local}
    Position=Multiply(A=@n44.PS, B=@n45) {constB=0.3}
   ) {materialFunction=&MF_Season40PickaxeTriplanarUVs}.Triplanar UVs) {name=TriPlanarUVForHighlights, nodeColor=<rgba 0.5469 0 1 1>}
  n46 = VertexInterpolator(VS=LocalPosition() {localOrigin=InstancePreSkinning}.XYZ)
  n47 = Multiply(
   A=MaterialFunctionCall(
    R=ScalarParameter() {defaultValue=1, group=Season40PickaxeReloaded, parameterName=MixGradientXScale}
    G=ScalarParameter() {defaultValue=1.5, group=Season40PickaxeReloaded, parameterName=MixGradientYScale}
    B=ScalarParameter() {defaultValue=0.2607, group=Season40PickaxeReloaded, parameterName=MixGradientZScale}
   ) {materialFunction=&AppendMany}.RGB
   B=Add(B=Constant() {r=0.2}) {constA=1}
  )
  n48 = TransformPosition(None=ObjectPositionWS()) {transformSourceType=TRANSFORMPOSSOURCE_World}
  n49 = VectorParameter() {parameterName=MixGradientPositionOffset, defaultValue=<rgba 0 7.2966 2.5311 0>}
  NamedRerouteDeclaration(None=SphereMask(
    A=Multiply(A=@n46.PS, B=@n47)
    B=Add(A=@n48, B=@n49.RGB)
    Radius=ScalarParameter() {defaultValue=24, parameterName=MixGradientMaskRadius}
    Hardness=ScalarParameter() {defaultValue=0.7876, parameterName=MixGradientMaskHardness}
   )) {name=MixGradientMask, nodeColor=<rgba 0 0.2311 1 1>}
  n50 = Add(A=NamedRerouteUsage(), B=Multiply(A=NamedRerouteUsage()) {constB=0.5})
  n51 = NamedRerouteUsage()
  NamedRerouteDeclaration(None=LinearInterpolate(A=NamedRerouteUsage(), B=Normalize(VectorInput=@n50), Alpha=ComponentMask(None=@n51) {g=true})) {name=Out_Normal, nodeColor=<rgba 1 0 0.2578 1>}
  NamedRerouteDeclaration(None=@n12.Random Grid) {name=RandomGrid, nodeColor=<rgba 0.5 0 1 1>}
  n52 = MaterialFunctionCall() {materialFunction=&BoundingBoxBased_0-1_UVW}
  n53 = SphereMask(
   A=MaterialFunctionCall(X=@n52.R, Y=@n52.R) {materialFunction=&MakeFloat2}.Result
   B=Constant() {r=0.5}
  ) {attenuationRadius=0.9, hardnessPercent=0}
  n54 = ScalarParameter() {group=STGridFX, parameterName=BandLocalPositionOffsetX}
  n55 = ScalarParameter() {group=STGridFX, parameterName=BandLocalPositionOffsety}
  n56 = Divide(
   A=Add(A=LocalPosition().XY, B=AppendVector(A=@n54, B=@n55))
   B=ScalarParameter() {defaultValue=100, group=STGridFX, parameterName=BandSpreadingLocalPositionDivider}
  ) {constB=100}
  n57 = Constant()
  n58 = Add(
   A=LinearInterpolate(
    A=OneMinus(None=@n53)
    B=SphereMask(A=@n56, B=@n57) {attenuationRadius=1.5, hardnessPercent=0}
    Alpha=ScalarParameter() {group=STGridFX, parameterName=BandSpreadingUseLocalPosition}
   )
   B=Multiply(A=NamedRerouteUsage(), B=ScalarParameter() {defaultValue=0.15, group=STGridFX, parameterName=BandSpreadingNoiseAmount}) {constB=0.1}
  )
  n59 = Time()
  n60 = ScalarParameter() {defaultValue=0.35, group=STGridFX, parameterName=BandSpreadingSpeed}
  n61 = Multiply(A=Frac(None=Multiply(A=@n59, B=@n60)), B=ScalarParameter() {defaultValue=1.3, group=STGridFX, parameterName=BandLocationMAX}) {constB=1.1}
  n62 = ScalarParameter() {defaultValue=0.15, group=STGridFX, parameterName=BandWidth}
  n63 = Multiply(
   A=Power(Base=SphereMask(A=@n58, B=@n61, Radius=@n62) {attenuationRadius=0.1, hardnessPercent=0})
   B=ScalarParameter() {defaultValue=2, group=STGridFX, parameterName=BandEmissiveMult}
  )
  n64 = NamedRerouteUsage()
  NamedRerouteDeclaration(None=MaterialFunctionCall(A=NamedRerouteUsage(), B=NamedRerouteUsage(), C=NamedRerouteUsage(), Alpha=LinearInterpolate(A=@n63, Alpha=@n64) {constB=0}) {materialFunction=&Lerp_3Color}.Result) {name=ExpandingWave, nodeColor=<rgba 0 1 0.1797 1>}
  NamedRerouteDeclaration(None=@n1.BaseColor) {name=In_BaseColor, nodeColor=<rgba 0.5469 0 1 1>}
  n65 = NamedRerouteUsage()
  NamedRerouteDeclaration(None=LinearInterpolate(A=NamedRerouteUsage(), B=NamedRerouteUsage(), Alpha=ComponentMask(None=@n65) {g=true})) {name=Out_BaseColor, nodeColor=<rgba 0 1 0.2266 1>}
  n66 = Transform(None=VertexNormalWS()) {transformSourceType=TRANSFORMSOURCE_World, transformType=TRANSFORM_Local}
  n67 = MaterialFunctionCall(
   R=ScalarParameter() {defaultValue=1, group=Season40PickaxeReloaded, parameterName=XTriPlanarNormalScale}
   G=ScalarParameter() {defaultValue=1, group=Season40PickaxeReloaded, parameterName=YTriPlanarNormalScale}
   B=ScalarParameter() {defaultValue=1, group=Season40PickaxeReloaded, parameterName=ZTriPlanarNormalScale}
  ) {materialFunction=&AppendMany}
  n68 = LocalPosition() {localOrigin=InstancePreSkinning}
  n69 = ScalarParameter() {defaultValue=1, group=Season40PickaxeReloaded, parameterName=XTriPlanarScale}
  n70 = ScalarParameter() {defaultValue=1, group=Season40PickaxeReloaded, parameterName=YTriPlanarScale}
  n71 = ScalarParameter() {defaultValue=1, group=Season40PickaxeReloaded, parameterName=ZTriPlanarScale}
  n72 = Add(
   A=Multiply(
    A=VertexInterpolator(VS=@n68.XYZ).PS
    B=MaterialFunctionCall(R=@n69, G=@n70, B=@n71) {materialFunction=&AppendMany}.RGB
   )
   B=VectorParameter() {group=Season40PickaxeReloaded, parameterName=TriplanarUVOffset, defaultValue=<rgba -0.52 0 0 1>}.RGB
  )
  NamedRerouteDeclaration(None=MaterialFunctionCall(Normal=Multiply(A=@n66, B=@n67.RGB), Position=Multiply(A=@n72) {constB=0.03}) {materialFunction=&MF_Season40PickaxeTriplanarUVs}.Triplanar UVs) {name=TriPlanarUVs, nodeColor=<rgba 0.5469 0 1 1>}
  n73 = TextureSample() {samplerType=SAMPLERTYPE_LinearColor, texture=/BRCosmetics/Weapons/FORT_Melee/Pickaxe_Season_40Reload/Textures/T_Pickaxe_Season40_Reload_FX2}
  NamedRerouteDeclaration(None=FunctionInput(Preview=@n73.RGB) {inputName=In_TechArtMask2, sortPriority=1}) {name=In_TechArtMask2, nodeColor=<rgba 0.2422 1 0 1>}
  n74 = LinearInterpolate(A=NamedRerouteUsage(), B=NamedRerouteUsage(), Alpha=ComponentMask(None=NamedRerouteUsage()) {r=true})
  n75 = NamedRerouteUsage()
  n76 = ComponentMask(None=NamedRerouteUsage()) {g=true}
  n77 = NamedRerouteUsage()
  n78 = NamedRerouteUsage()
  NamedRerouteDeclaration(None=Add(A=LinearInterpolate(A=@n74, B=@n75, Alpha=@n76), B=Multiply(A=@n77, B=@n78))) {name=LerpedUVs, nodeColor=<rgba 1 0 0.0937 1>}
  n79 = TextureCoordinate() {desc=UV for circular side piece}
  n80 = AppendVector(
   A=ScalarParameter() {defaultValue=1, group=Season40PickaxeReloaded, parameterName=SideCircleXScale}
   B=ScalarParameter() {defaultValue=1, group=Season40PickaxeReloaded, parameterName=SideCircleYScale}
  )
  n81 = ScalarParameter() {group=Season40PickaxeReloaded, parameterName=SideCircleXOffset}
  n82 = ScalarParameter() {group=Season40PickaxeReloaded, parameterName=SideCircleYOffset}
  NamedRerouteDeclaration(None=Add(A=Multiply(A=@n79, B=@n80), B=AppendVector(A=@n81, B=@n82))) {name=SideCircleUVs, nodeColor=<rgba 0 0.4141 1 1>}
  NamedRerouteUsage()
 ]}
