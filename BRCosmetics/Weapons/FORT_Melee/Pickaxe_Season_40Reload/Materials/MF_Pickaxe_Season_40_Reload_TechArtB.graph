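# Material function for the Season 40 Reload pickaxe that blends an animated "Ranked Reload" grid effect over incoming material attributes within a channel-selected mask.
# MF_AnimatedGridTexture builds the colored and grayscale grid, and derived values override emissive, base color, metallic, specular, roughness and opacity.
# A looping radial wipe sphere-masks time against distance from a local-space center (pre-skinning position via vertex interpolator), with noise and grid warping.
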MaterialFunction /BRCosmetics/Weapons/FORT_Melee/Pickaxe_Season_40Reload/Materials/MF_Pickaxe_Season_40_Reload_TechArtB
graph {nodes [
  n1 = NamedRerouteUsage()
  n2 = SetMaterialAttributes!(
   MaterialAttributes=@n1
   EmissiveColor=NamedRerouteUsage()
   BaseColor=Multiply(A=NamedRerouteUsage(), B=ScalarParameter() {defaultValue=0.3, group=Ranked Reload, parameterName=RankedReload.BaseColor_GridAmount})
   Metallic=ScalarParameter() {defaultValue=0.7, group=Ranked Reload, parameterName=RankedReload.Metallic}
   Specular=NamedRerouteUsage()
   Roughness=NamedRerouteUsage()
   Opacity=ScalarParameter() {defaultValue=1, group=Ranked Reload, parameterName=RankedReload.Opacity, sortPriority=500}
  ) {attributeSetTypes [
    #1
    #2
    #3
    #4
    #5
    #6
   ]}
  n3 = ChannelMaskParameter(None=NamedRerouteUsage()) {group=Ranked Reload, maskChannel=Green, parameterName=RankedReload.MaskChannel, defaultValue=<rgba 0 1 0 0>}
  FunctionOutput(None=BlendMaterialAttributes(A=@n1, B=@n2, Alpha=NamedRerouteDeclaration(None=@n3.R) {name=RankedReloadMask, nodeColor=<rgba 0 1 0.5781 1>}))
  n4 = Time()
  n5 = ScalarParameter() {defaultValue=0.3, group=Ranked Reload, parameterName=RankedReload.Wipe_Speed}
  n6 = Multiply(A=Frac(None=Multiply(A=@n4, B=@n5)), B=ScalarParameter() {defaultValue=100, group=Ranked Reload, parameterName=RankedReload.Wipe_Distance}) {constB=100}
  n7 = VectorParameter() {group=Ranked Reload, parameterName=RankedReload.Wipe_CenterPos, defaultValue=<rgba 0 0 108 1>}
  n8 = VertexInterpolator(VS=LocalPosition() {localOrigin=InstancePreSkinning}.XYZ)
  n9 = Add(A=Length(None=Subtract(A=@n7.RGB, B=@n8.PS)) {is=class MaterialExpressionLength}, B=NamedRerouteUsage())
  n10 = ScalarParameter() {defaultValue=8, group=Ranked Reload, parameterName=RankedReload.Wipe_Radius}
  n11 = ScalarParameter() {group=Ranked Reload, parameterName=RankedReload.Wipe_Hardness}
  NamedRerouteDeclaration(None=Power(
    Base=SphereMask(A=@n6, B=@n9, Radius=@n10, Hardness=@n11)
    Exp=ScalarParameter() {defaultValue=2, group=Ranked Reload, parameterName=RankedReload.Wipe_Exponent}
   )) {name=Wipe, nodeColor=<rgba 1 0.7031 0 1>}
  n12 = TextureSampleParameter2D() {
   group=Ranked Reload
   parameterName=RankedReload.GridTexture
   samplerType=SAMPLERTYPE_Masks
   sortPriority=0
   texture=/Game/Global/Textures/Default/Blanks/T_White
  }
  n13 = NamedRerouteUsage()
  n14 = VectorParameter() {group=Ranked Reload, parameterName=RankedReload.Color_01, sortPriority=2, defaultValue=<rgba 0.05 0.25 0.35 1>}
  n15 = VectorParameter() {group=Ranked Reload, parameterName=RankedReload.Color_02, sortPriority=3, defaultValue=<rgba 0 0.65 1 1>}
  n16 = VectorParameter() {group=Ranked Reload, parameterName=RankedReload.LineColor, sortPriority=4, defaultValue=<rgba 0 0.75 1.5 1>}
  n17 = MaterialFunctionCall!(
   Grid=@n12.RGB
   Grid Lines=NamedRerouteDeclaration(None=ChannelMaskParameter(None=@n13) {group=Ranked Reload, parameterName=RankedReload.GridLineChannel, sortPriority=1}.R) {name=GridLines, nodeColor=<rgba 1 0.3281 0 1>}
   Color 01=@n14.RGB
   Color 02=@n15.RGB
   Speed=ScalarParameter() {defaultValue=0.2, group=Ranked Reload, parameterName=RankedReload.Speed, sortPriority=5}
   Speed Random Amount=ScalarParameter() {defaultValue=2, group=Ranked Reload, parameterName=RankedReload.Speed_RandomAmount, sortPriority=6}
   Line Color=@n16.RGB
  ) {materialFunction=/Game/Global/MaterialFunctions/Cosmetics/AnimatedGrid/MF_AnimatedGridTexture}
  NamedRerouteDeclaration(None=@n17."Result (V3)") {name=ColorGrid, nodeColor=<rgba 1 0.3516 0 1>}
  NamedRerouteDeclaration(None=@n17.Grayscale Grid) {name=GrayscaleGrid, nodeColor=<rgba 0.0312 0 1 1>}
  n18 = NamedRerouteUsage()
  n19 = VectorParameter() {group=Ranked Reload, parameterName=RankedReload.Wipe_Color, defaultValue=<rgba 0.8594 2 5 1>}
  n20 = Multiply(A=NamedRerouteUsage(), B=ScalarParameter() {defaultValue=0.25, group=Ranked Reload, parameterName=RankedReload.WipeEmissive_Strength})
  n21 = ScalarParameter() {defaultValue=0.85, group=Ranked Reload, parameterName=RankedRelaoad.Emissive_Min}
  n22 = ScalarParameter() {defaultValue=0.85, group=Ranked Reload, parameterName=RankedRelaoad.Emissive_Max}
  n23 = Power(
   Base=OneMinus(None=NamedRerouteUsage())
   Exp=ScalarParameter() {defaultValue=1, group=Ranked Reload, parameterName=RankedRelaoad.Emissive_FresnelExponent}
  )
  NamedRerouteDeclaration(None=Multiply(A=LinearInterpolate(A=@n18, B=@n19.RGB, Alpha=@n20), B=LinearInterpolate(A=@n21, B=@n22, Alpha=@n23))) {name=OUT_Emissive, nodeColor=<rgba 0.2891 1 0 1>}
  n24 = Subtract(A=NamedRerouteUsage(), B=NamedRerouteUsage())
  NamedRerouteDeclaration(None=LinearInterpolate(
    A=ScalarParameter() {group=Ranked Reload, parameterName=RankedReload.Roughness_Min}
    B=ScalarParameter() {defaultValue=1, group=Ranked Reload, parameterName=RankedReload.Roughness_Max}
    Alpha=Saturate(None=@n24)
   )) {name=OUT_Roughness, nodeColor=<rgba 1 0.4922 0 1>}
  n25 = Add(A=NamedRerouteUsage(), B=NamedRerouteUsage())
  NamedRerouteDeclaration(None=LinearInterpolate(
    A=ScalarParameter() {group=Ranked Reload, parameterName=RankedReload.Spec_Min}
    B=ScalarParameter() {defaultValue=1, group=Ranked Reload, parameterName=RankedReload.Spec_Max}
    Alpha=Saturate(None=@n25)
   )) {name=OUT_Spec, nodeColor=<rgba 1 0.4922 0 1>}
  n26 = TextureCoordinate()
  n27 = ScalarParameter() {defaultValue=2, group=Ranked Reload, parameterName=RankedReload.Wipe_WarpNoiseTile}
  n28 = ChannelMaskParameter(None=TextureSample(UVs=Multiply(A=@n26, B=@n27)) {samplerType=SAMPLERTYPE_Masks, texture=/Game/Global/Textures/Default/Noise/T_Generic_Perlin}.RGB) {group=Ranked Reload, parameterName=RankedReload.Wipe_WarpNoiseChannel}
  n29 = ScalarParameter() {defaultValue=5, group=Ranked Reload, parameterName=RankedReload.Wipe_WarpNoiseStrength}
  n30 = ScalarParameter() {defaultValue=15, group=Ranked Reload, parameterName=RankedReload.Wipe_WarpGridStrength}
  n31 = NamedRerouteUsage()
  NamedRerouteDeclaration(None=Add(A=Multiply(A=@n28.R, B=@n29), B=Multiply(A=@n30, B=@n31))) {name=WipeWarp, nodeColor=<rgba 0.1016 1 0 1>}
  NamedRerouteDeclaration(None=FunctionInput() {inputName=In_MaterialAttributes, inputType=FunctionInput_MaterialAttributes}) {name=In_MaterialAttributes, nodeColor=<rgba 0.0781 0 1 1>}
  n32 = TextureSample() {samplerType=SAMPLERTYPE_LinearColor, texture=/BRCosmetics/Weapons/FORT_Melee/Pickaxe_Season_40Reload/Textures/T_Season_40Reload_Pickaxe_FX}
  NamedRerouteDeclaration(None=FunctionInput(Preview=@n32.RGB) {inputName=In_TechArtMask, sortPriority=1}) {name=In_TechArtMask, nodeColor=<rgba 0.2422 1 0 1>}
  NamedRerouteDeclaration(None=FunctionInput() {inputName=In_Fresnel, inputType=FunctionInput_Scalar, sortPriority=2}) {name=In_Fresnel, nodeColor=<rgba 0.2891 0 1 1>}
 ]}
